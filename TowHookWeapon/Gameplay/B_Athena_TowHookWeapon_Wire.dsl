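; Rope actor for the tow hook weapon that animates the wire reeling back in.
; When OnEndReel fires, a timeline moves the hook projectile back to the start, stretching, aiming and wiggling the rope mesh by material parameters.
; Finishing plays retract FX, hides the rope, sends the pawn a stop gameplay event and destroys the actor.
; Server and client finish through separate DoOnce-gated events, a noted workaround for unreliable client timeline completion.
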(event Custom|ReelFinished
  ;[1 1 1] When the reel is finish play FX, Hide the rope, and tell the rope to drop the items and destroy
  (Utilities|FlowControl|DoOnce false
    (:Completed
      (Development|PrintString "Reel Finished" true true "(R=1.000000,G=0.126427,B=0.246205,A=1.000000)")
      (Niagara|SpawnSystemAttached (Variables|Default|GetRetract%20NS%20System) (Variables|Weapon|GetWeaponMesh) "muzzle" "0, 0, 0" "0, 0, 0" "KeepRelativeOffset" true true "AutoRelease")
      (Development|SetHiddeninGame (Variables|StaticMesh|GetRopeMesh) true false)
      (Ability|SendGameplayEventtoActor (Variables|Default|GetPlayerPawn) (Variables|Default|GetT_Stop) :!pinned true) ; Plays reattach anim
      (Utilities|FlowControl|Delay 0.2)
      (Actor|DestroyActor))))

(event Custom|ReelFinishedClientOnly
  ;[1 1 1] When the reel is finish play FX, Hide the rope, and tell the rope to drop the items and destroy
  (Utilities|FlowControl|DoOnce false
    (:Completed
      (Development|PrintString "Reel Finished" true true "(R=1.000000,G=0.126427,B=0.246205,A=1.000000)")
      (Niagara|SpawnSystemAttached (Variables|Default|GetRetract%20NS%20System) (Variables|Weapon|GetWeaponMesh) "muzzle" "0, 0, 0" "0, 0, 0" "KeepRelativeOffset" true true "AutoRelease")
      (Development|SetHiddeninGame (Variables|StaticMesh|GetRopeMesh) true false)
      (Ability|SendGameplayEventtoActor (Variables|Default|GetPlayerPawn) (Variables|Default|GetT_Stop) :!pinned true) ; Plays reattach anim
      (Utilities|FlowControl|Delay 0.2)
      (Actor|DestroyActor))))

(event EventBeginPlay
  ;[1 1 1] EVENT -> BeginPlay
  (EventDispatchers|BindEventtoOnEndReel (EventDispatchers|CreateEvent :!event Custom|End%20Reel)))

(event Custom|End%20Reel
  (bind _ropeprojectilemesh (Variables|StaticMesh|GetRopeProjectileMesh))
  (bind _returnvalue (Transformation|GetSocketLocation _ropeprojectilemesh "End"))
  (bind _startlocation (Variables|FortHookGunRope|GetStartLocation))
  (bind _ropemesh (Variables|StaticMesh|GetRopeMesh))
  (bind _ropemesh_2 (Variables|StaticMesh|GetRopeMesh))
  (bind _endlocation (Variables|FortHookGunRope|GetEndLocation))
  (bind _startlocation_2 (Variables|FortHookGunRope|GetStartLocation))
  ;[1 1 1] Reel In
  (Utilities|FlowControl|DoOnce false
    (:Completed
      (Components|Timeline|SetPlayRate (Variables|B_Athena_TowHookWeapon_Wire|GetReel%20In) (Variables|Default|GetReelinPlayRate))
      (bind (_direction _reel_lerp_controller _tighten) (Timeline|Reel%20In PlayFromStart)
        (:Update
          (CallFunction|SetStartPointOfRope)
          (Transformation|SetWorldLocationAndRotation (Variables|StaticMesh|GetRopeProjectileMesh) (Math|Vector|Lerp(Vector) _endlocation _startlocation _reel_lerp_controller) (Math|Rotator|FindLookatRotation _startlocation _endlocation) false false)
          (bind _distance (Class|MissionBlueprintFunctionLibrary|DistanceBetweenTwoVectors _startlocation_2 (- (Math|Vector|Lerp(Vector) _endlocation _startlocation _reel_lerp_controller) (- (Math|Vector|Lerp(Vector) _endlocation _startlocation _reel_lerp_controller) _returnvalue))))
          (Transformation|SetRelativeScale3D _ropemesh (Math|Vector|MakeVector (/ _distance 1e+02) 1.0 1.0))
          (Transformation|SetWorldRotation _ropemesh (Math|Rotator|FindLookatRotation _startlocation_2 (- (Math|Vector|Lerp(Vector) _endlocation _startlocation _reel_lerp_controller) (- (Math|Vector|Lerp(Vector) _endlocation _startlocation _reel_lerp_controller) _returnvalue))) false false)
          (Rendering|Material|SetScalarParameterValueonMaterials _ropemesh_2 "ReturnWiggle" (Math|Float|Lerp 0.0 (ScalableFloat|GetValueatLevel (Variables|HF|GetReturnWiggle)) _tighten))
          (Rendering|Material|SetScalarParameterValueonMaterials _ropemesh_2 "WiggleAmount" (Math|Float|Lerp 0.0 (ScalableFloat|GetValueatLevel (Variables|HF|GetReturnWiggleAmount)) _tighten)))
        (:Finished
          ;;;[1 1 1] Sometimes the client is not firing off the event. Try to multicast it and also send it on the client. We are gating the other function behind a do once so that we only play it once on the client. This is just so that it gets to the client faster in the case that we did not lose it.
          ;;[1 0.6907 0] Need to check with engineering why the reel timeline end may not be playing on the client. It is hitting the timeline, just never firing finished.
          (if (Networking|IsServer)
            (CallFunction|ReelFinished)
            (else
              (CallFunction|ReelFinishedClientOnly))))))))

(event Custom|TightenRope
  ;[1 1 1] Tighten Rope
  (bind (_direction _newtrack_0) (Timeline|Tight PlayFromStart)
    (:Update
      (Rendering|Material|SetScalarParameterValueonMaterials (Variables|StaticMesh|GetRopeMesh) "WiggleAmount" (Math|Float|Lerp 0.0 15.0 _newtrack_0)))))

(fn ConstructionScript)
