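; Fortnite tow hook weapon that fires a projectile which can lock onto vehicles and tether the player to what it hits.
; A locally viewed timer picks the auto-target and reticle state, passing the target to the server for projectile homing.
; Replicated hit data starts tethered movement on non-authority instances and picks the montage section from the tethered actor's velocity.
; It also drives fire, wire and impact audio, and toggles HUD keys from the tethered tag.
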(event EventEndPlay
  (params (EndPlayReason :type EEndPlayReason%20Enum :default "Destroyed"))
  ;[1 0.5272 0.04774] EventEndPlay
  ; Final cleansup of BP: Unbind, Clear timers, etc.
  (|CallParentFunction EndPlayReason))

(event Custom|ForceTowHookStopMovementMode
  (bind _asfort_player_pawn (Utilities|Casting|CastToFortPlayerPawn (Game|GetInstigator)))
  (Class|FortPawnComponentTetheredMovement|EndTetheredMovement (Actor|GetComponentbyClass _asfort_player_pawn "/Script/TowHookWeaponRuntime.FortPawnComponent_TetheredMovement")))

(event EventBeginPlay
  ;[0.1083 1 0.0636] EventBeginPlay
  ; Bind events and setup
  (|CallParentFunction)
  (bind _asfort_player_pawn (Utilities|Casting|CastToFortPlayerPawn (Game|GetInstigator)))
  (Variables|Default|SetFort%20Player%20Pawn _asfort_player_pawn))

(event FortWeapon|EventOnEquippedAndAttached
  ;[0 0.2176 1] On Equipped - start async watcher for tethered state & targeting timer
  (|CallParentFunction)
  (Utilities|FlowControl|Sequence
    (:then_0
      (bind _asfort_player_pawn (Utilities|Casting|CastToFortPlayerPawn (Game|GetInstigator)))
      (bind _asynctaskproxy (Ability|Tasks|WaitGameplayTagAddToActor _asfort_player_pawn (Variables|Tags|GetTethered%20Tag) true)
        (:then
          (Variables|Async|SetTethered%20Async%20Action _asynctaskproxy)
          (if (not (GameplayTags|HasMatchingGameplayTag _asfort_player_pawn (Variables|Tags|GetTethered%20Tag)))
            (CallFunction|UpdateHUDKeys false)))
        (:Added
          (Ability|Async|EndAction (Variables|Async|GetTethered%20Async%20Action))
          (CallFunction|UpdateHUDKeys true))))
    (:then_1
      (bind _asfort_player_pawn_2 (Utilities|Casting|CastToFortPlayerPawn (Game|GetInstigator)))
      (if (Pawn|IsLocallyViewed _asfort_player_pawn_2)
        (Variables|Targeting|SetbAllowAutoTargeting (ScalableFloat|GetValueasBool (Variables|Targeting|GetAuto%20Targeting%20Enabled)))
        (bind _returnvalue (Utilities|Time|SetTimerbyEvent (delegate Custom|Event%20Update%20Reticle%20Targeting) (Variables|Targeting|GetUpdateReticleTargetingRate) true true))
        (Variables|Targeting|SetReticle%20Targeting%20Timer _returnvalue)))))

(event Custom|Event%20Update%20Reticle%20Targeting
  (Default|UpdateReticleTargeting))

(event FortWeapon|EventOnUnEquippedAndDetached
  (Ability|Async|EndAction (Variables|Async|GetTethered%20Async%20Action))
  (CallFunction|UpdateHUDKeys false))

(event Custom|Event%20ServerSetTargetHomingActor
  (params (New%20Target :type Actor%20Object%20Reference))
  (bind _output_get (Variables|Targeting|SetTargetHomingActor New%20Target))
  (Development|PrintString (Utilities|GetDisplayName _output_get)))

(event FortWeapon|EventOnSpawnProjectile
  (params (SpawnProjectile :type Fort%20Projectile%20Base%20Object%20Reference))
  (bind _targethomingactor (Variables|Targeting|GetTargetHomingActor))
  (Networking|SwitchHasAuthority
    (:Authority
      (Utilities|IsValid _targethomingactor
        (:Is%20Valid
          (Projectile|SetHomingData SpawnProjectile SpawnProjectile (Utilities|Struct|MakeProjectileHomingData "LockOn" 999999.0 999999.0 0.5 0.0 -1 _targethomingactor 0 "0.000000,0.000000,0.000000" (* (Math|Vector|VectorUp) (ScalableFloat|GetValueatLevel (Variables|Targeting|GetTargetVehicleOffset))) "Relative") :!class /Script/FortniteGame.FortProjectileMovementInterface))))))

(event EventPlayWeaponFireCustomAudio
  (params (bPersistentFire :type Boolean) (bSecondaryFire :type Boolean))
  ;[0.04833 0.0575 0.1406] Play Fire sound 1P and 3P, with last shot layer and wire whip sound. Wire whip sound gets faded out when the projectile hits something.
  (Utilities|FlowControl|Sequence
    (:then_0
      (bind _fort_player_pawn (Variables|Default|GetFort%20Player%20Pawn)
        (:then
          (Variables|Default|SetIs%20Locally%20Viewed (Pawn|IsLocallyViewed _fort_player_pawn)))
        (:else)))
    (:then_1
      (Utilities|FlowControl|Sequence
        (:then_0
          (bind _returnvalue (Audio|SpawnSoundAttached (select (Variables|Default|GetIs%20Locally%20Viewed) (Variables|Audio|GetFire%201P%20Sound) (Variables|Audio|GetFire%203P%20Sound)) (Variables|Transformation|GetRootComponent)))
          (Variables|Default|SetFire%20Audio _returnvalue))
        (:then_1
          (Utilities|FlowControl|Sequence
            (:then_0
              (Audio|Parameter|SetBooleanParameter (Variables|Default|GetFire%20Audio) "bIsLastShot" (<= (FortWeapon|GetRemainingAmmo) 1) :!class /Script/AudioExtensions.AudioParameterControllerInterface)
              (Audio|Parameter|ExecuteTriggerParameter (Variables|Default|GetFire%20Audio) "Events.Fire"))))))))

(event Custom|On%20Projectile%20Hit
  (params (Hit%20Actor :type Actor%20Object%20Reference) (Projectile :type Fort%20Projectile%20Base%20Object%20Reference))
  ;[0.224 0.09519 0.06151] Logic for stopping the wire whip layer of the Fire sound and playing enemy impact sounds. Projectile references are cached to prevent sound from playing more than once per impact.
  (Variables|Default|SetNewProjectile Projectile)
  (Networking|SwitchHasAuthority
    (:Remote
      (Utilities|FlowControl|Sequence
        (:then_0
          (bind _fire_audio (Variables|Default|GetFire%20Audio)
            (:then
              (Audio|Parameter|ExecuteTriggerParameter _fire_audio "Events.StopWire"))
            (:else)))
        (:then_1
          (Utilities|FlowControl|Sequence
            (:then_0
              (if (!= (Variables|Default|GetNewProjectile) (Variables|Default|GetProjectileCache))
                (bind _asfort_player_pawn (Utilities|Casting|CastToFortPlayerPawn Hit%20Actor))
                (Audio|SpawnSoundAttached (select (Pawn|IsLocallyViewed (Game|GetInstigator)) (Variables|Audio|GetEnemy%20Impact%20Sound%201P) (Variables|Audio|GetEnemy%20Impact%20Sound%203P)) (Variables|/Script/Engine.Actor|GetRootComponent _asfort_player_pawn))
                (Variables|Default|SetProjectileCache (Variables|Default|GetNewProjectile))
                ;[0.3073 0.07202 0.2128] Duck the fire sound for 1P when hitting an enemy player
                (if (Pawn|IsLocallyViewed (Game|GetInstigator))
                  (Audio|Modulation|ActivateControlBusMix (Variables|Audio|GetControl%20Bus%20Mix))
                  (Utilities|FlowControl|Delay 0.3)
                  (Audio|Modulation|DeactivateControlBusMix (Variables|Audio|GetControl%20Bus%20Mix)))))))))))

(fn DetermineAttenuation
  (params (Actor :type Actor%20Object%20Reference))
  (return (select (and (Utilities|Enum|Equal(Enum) (Team|GetTeamAffiliationforActortoLocalPlayer Actor) "Hostile") (Pawn|IsLocallyViewed (Game|GetInstigator))) (Variables|Audio|Get1P%20Attenuation%20Settings) 0)))

(fn GetMovementRuntimeData
  (bind _returnvalue (Game|GetInstigator))
  (bind _returnvalue_2 (Actor|GetComponentbyClass _returnvalue "/Script/TowHookWeaponRuntime.FortPawnComponent_TetheredMovement"))
  (Utilities|IsValid _returnvalue_2
    (:Is%20Valid
      (return (Class|FortPawnComponentTetheredMovement|GetMovementRuntimeData _returnvalue_2)))
    (:Is%20Not%20Valid
      (return 0))))

(fn OnRep_HideProj)

(fn OnRep_ProjectileHitData
  (bind _runtime_data (CallFunction|GetMovementRuntimeData))
  (bind _nettetheredactor (Variables|/Script/TowHookWeaponRuntime.FortMovementMode_TetheredRuntimeData|GetNetTetheredActor _runtime_data))
  (bind (_returnvalue_x _returnvalue_y _returnvalue_z) (Math|Vector|BreakVector (Transformation|GetVelocity _nettetheredactor)))
  (bind _returnvalue (Math|Float|Absolute(Float) _returnvalue_z))
  (bind _returnvalue_2 (Math|Float|Absolute(Float) _returnvalue_x))
  (bind _returnvalue_3 (Math|Float|Absolute(Float) _returnvalue_y))
  (bind _returnvalue_4 (Utilities|Operators|Add _returnvalue_2 _returnvalue_3 _returnvalue))
  (bind (_projectilehitdata_projectile_9_ff96757148c1260dff8e428a8db9d153 _projectilehitdata_hitactor_7_424357ae4112f32cb1760fb82d90f298 _projectilehitdata_localhitlocation_10_41bbf2064bde9c04d5f1d9a7d835fd6e) (Utilities|Struct|BreakTowHookProjectileHit (Variables|GetProjectileHitData_Projectile_9_FF96757148C1260DFF8E428A8DB9D153)))
  (bind _projectilehitdata_projectile_9_ff96757148c1260dff8e428a8db9d153_2 (Utilities|Struct|BreakTowHookProjectileHit (Variables|GetProjectileHitData_Projectile_9_FF96757148C1260DFF8E428A8DB9D153)))
  (bind (_projectilehitdata_projectile_9_ff96757148c1260dff8e428a8db9d153_3 _projectilehitdata_hitactor_7_424357ae4112f32cb1760fb82d90f298_2) (Utilities|Struct|BreakTowHookProjectileHit (Variables|GetProjectileHitData_Projectile_9_FF96757148C1260DFF8E428A8DB9D153)))
  (Networking|SwitchHasAuthority
    (:Remote
      (Utilities|FlowControl|Sequence
        (:then_0
          (Utilities|IsValid _projectilehitdata_projectile_9_ff96757148c1260dff8e428a8db9d153_2
            (:Is%20Valid
              (Class|PrjAthenaTowHookWeapon|StopLoopingAudioTell _projectilehitdata_projectile_9_ff96757148c1260dff8e428a8db9d153_2)
              (Class|PrjAthenaTowHookWeapon|StopLoopingNSTrail _projectilehitdata_projectile_9_ff96757148c1260dff8e428a8db9d153_2)
              ;[0.1466 0 1] GCNL - projectile looping to help ppl in vehicle be aware they are hooked. Self stopped when projectile is killed
              (bind _asfort_athena_vehicle (Utilities|Casting|CastToFortAthenaVehicle _projectilehitdata_hitactor_7_424357ae4112f32cb1760fb82d90f298_2))
              (GameplayCue|AddGameplayCueOnActor(Looping) _projectilehitdata_projectile_9_ff96757148c1260dff8e428a8db9d153_2 (Variables|Tags|GetGCNL%20Prj%20Attached%20Gameplay%20Cue%20Tag) (Ability|GameplayCue|MakeGameplayCueParameters :NormalizedMagnitude 0.0 :RawMagnitude 0.0 :Location "0,0,0" :Normal "0,0,0" :Instigator (Game|GetInstigator) :EffectCauser _asfort_athena_vehicle :SourceObject (Variables|Default|GetRope) :GameplayEffectLevel 1 :AbilityLevel 1 :bReplicateLocationWhenUsingMinimalRepProxy false)))))
        (:then_1
          (Utilities|IsValid _projectilehitdata_hitactor_7_424357ae4112f32cb1760fb82d90f298
            (:Is%20Valid
              (bind _asfort_player_pawn (Utilities|Casting|CastToFortPlayerPawn (Game|GetInstigator)))
              (Utilities|IsValid (Actor|GetComponentbyClass _asfort_player_pawn "/Script/TowHookWeaponRuntime.FortPawnComponent_TetheredMovement")
                (:Is%20Valid
                  (Class|FortPawnComponentTetheredMovement|StartTetheredMovement (Actor|GetComponentbyClass _asfort_player_pawn "/Script/TowHookWeaponRuntime.FortPawnComponent_TetheredMovement") (Utilities|Struct|MakeFortMovementModeTetheredCreationData _projectilehitdata_hitactor_7_424357ae4112f32cb1760fb82d90f298 _projectilehitdata_localhitlocation_10_41bbf2064bde9c04d5f1d9a7d835fd6e _projectilehitdata_projectile_9_ff96757148c1260dff8e428a8db9d153))
                  ;[0.4583 0.222 0.3448] KJHawkinson - Moved section picking logic from AnimBP to avoid delay from other clients view
                  (if (> _returnvalue_4 2e+03)
                    (Variables|Default|SetStart%20Section%20Name "Extreme")
                    (elif (< _returnvalue_4 5.0)
                      (Variables|Default|SetStart%20Section%20Name "Stationary")
                      (else
                        (Variables|Default|SetStart%20Section%20Name "Default"))))
                  (Animation|Utilities|PlayAnimMontage (Game|GetInstigator) (Variables|Default|GetStart%20Tether%20Anim%20Montage) 1.0 (Variables|Default|GetStart%20Section%20Name)))))))))))

(fn PlayReticleSound
  ;[0 0.1341 0.5521 1 14] Event dispatcher is currently being called every time the reticle state is evaluated—not just when the state changes. So for now we will cache it and only play a sound when the state changes.
  (if (Utilities|Enum|NotEqual(Enum) (Variables|Targeting|GetReticleState) (Variables|Default|GetReticle%20State%20Cache))
    (Variables|Default|SetReticle%20State%20Cache (Variables|Targeting|GetReticleState))))

(fn SetTargetHomingActor
  (params (New%20Target :type Actor%20Object%20Reference))
  (if (!= New%20Target (Variables|Targeting|GetTargetHomingActor))
    (bind _output_get (Variables|Targeting|SetTargetHomingActor New%20Target))
    (CallFunction|EventServerSetTargetHomingActor _output_get)))

(fn UpdateHUDKeys
  (params (Tethered :type Boolean))
  (WeaponHUDData|SetHudKeyActionsVisibility (Utilities|Array|MakeArray (Utilities|Struct|MakeWeaponHudKeyActionVisibility 2 Tethered) (Utilities|Struct|MakeWeaponHudKeyActionVisibility 3 Tethered) (Utilities|Struct|MakeWeaponHudKeyActionVisibility 4 Tethered) (Utilities|Struct|MakeWeaponHudKeyActionVisibility 5 Tethered))))

(fn UpdateReticleTargeting
  (bind _asfort_player_pawn (Utilities|Casting|CastToFortPlayerPawn (Game|GetInstigator)))
  (bind _output_get (LocalVariables|SetLocal%20Player%20Pawn _asfort_player_pawn))
  (Utilities|IsValid _output_get
    (:Is%20Valid
      (bind _returnvalue (Class|TowHookWeaponBlueprintLibrary|GetTowHookWeaponTargetInfo _asfort_player_pawn (LocalVariables|GetTargetingConfig)))
      (Utilities|IsValid (.TargetActor (Utilities|Struct|BreakTowHookWeaponTargetInfo _returnvalue))
        (:Is%20Valid
          (Variables|Targeting|SetReticleWorldLocation (Math|Transform|TransformLocation (Transformation|GetActorTransform (.TargetActor (Utilities|Struct|BreakTowHookWeaponTargetInfo _returnvalue))) (* (Math|Vector|VectorUp) (Math|Float|SelectFloat (ScalableFloat|GetValueatLevel (Variables|Targeting|GetSF_ZOffset_Bus)) (ScalableFloat|GetValueatLevel (Variables|Targeting|GetTargetVehicleOffset)) (GameplayTags|MatchesAnyTags "(TagName=\"Vehicle.MoonFlaxWarBus\")" (Class|FortVehicleInterface|GetVehicleTags (Utilities|Casting|CastToFortVehicleInterface (.TargetActor (Utilities|Struct|BreakTowHookWeaponTargetInfo _returnvalue))))))))))
        (:Is%20Not%20Valid
          (Variables|Targeting|SetReticleWorldLocation "0, 0, 0")))
      (Utilities|Casting|CastToFortAthenaVehicle (.TargetActor (Utilities|Struct|BreakTowHookWeaponTargetInfo _returnvalue))
        (:then
          (if (Utilities|Struct|BreakTowHookWeaponTargetInfo _returnvalue)
            ;[0 0.02909 1] If player is aiming, disable auto-targeting
            (if (Variables|Targeting|GetbAllowAutoTargeting)
              (if (Variables|/Script/FortniteGame.FortPlayerPawn|GetbIsTargeting (LocalVariables|GetLocal%20Player%20Pawn))
                (Variables|Targeting|SetReticleState "NewEnumerator0")
                (CallFunction|SetTargetHomingActor 0)
                (else
                  (Variables|Targeting|SetReticleState "NewEnumerator2")
                  (CallFunction|SetTargetHomingActor (.TargetActor (Utilities|Struct|BreakTowHookWeaponTargetInfo _returnvalue)))))
              (Class|FortWeaponComponentTargetState|UpdateTargetInfo (Variables|Default|GetFortWeaponComponent_TargetState) (Utilities|Enum|NotEqual(Enum) (Variables|Targeting|GetReticleState) "NewEnumerator0") (Variables|Targeting|GetReticleWorldLocation))
              (Development|PrintString (Utilities|String|Append "Tow Hook Reticle: " (Utilities|String|EnumtoString (Variables|Targeting|GetReticleState)))))))
        (:CastFailed
          (Variables|Targeting|SetReticleState "NewEnumerator0"))))))

(fn ConstructionScript
  (|CallParentFunction))
