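; Tow hook weapon projectile that attaches to vehicles or trains so the firing player can be tethered.
; The server spawns a rope on launch, then range-checks and validates hits, attaching or killing the projectile, with timers for rope detach and failsafe cleanup.
; Vehicle attachment can eject the shooter from their current vehicle and passes hit data through the weapon.
; A delayed RepNotify workaround aligns the client hook mesh to the vehicle socket.
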(event Custom|AdjustPrjLifetime
  (bind _fortpawn (Variables|Default|GetFortPawn))
  (bind _returnvalue (Actor|GetComponentbyClass _fortpawn "/Script/TowHookWeaponRuntime.FortPawnComponent_TetheredMovement"))
  ;[1 0.3879 0] After 5s, if we're not attached to a vehicle or train, kill the projectile
  ; FORT-753580 - if tether ends due to Block MME tags in TetherExtLogic, we never hit any events there, so we the destroy actor call to the proj never happens, and we wind up here.
  ; Determine if we're not attached to anything and kill projectile
  (Utilities|IsValid (Variables|Hit|GetHit%20Actor)
    (:Is%20Valid
      (Utilities|Casting|CastToFortVehicleInterface (Variables|Hit|GetHit%20Actor)
        (:then
          _)
        (:CastFailed
          (bind _returnvalue_2 (GameplayTags|HasMatchingGameplayTag(Message) (Variables|Hit|GetHit%20Actor) (Variables|Tags|GetTrain%20Tag)))
          (if _returnvalue_2
            (Utilities|IsValid _fortpawn
              (:Is%20Valid
                (Utilities|IsValid _returnvalue
                  (:Is%20Valid
                    (if (or (Utilities|Enum|Equal(Enum) (Variables|/Script/TowHookWeaponRuntime.FortMovementMode_TetheredRuntimeData|GetState (Class|FortPawnComponentTetheredMovement|GetMovementRuntimeData _returnvalue)) "Active") (Utilities|Enum|Equal(Enum) (Variables|/Script/TowHookWeaponRuntime.FortMovementMode_TetheredRuntimeData|GetState (Class|FortPawnComponentTetheredMovement|GetMovementRuntimeData _returnvalue)) "Dormant"))
                      _
                      (else
                        _))))))
            (else
              _))
          (Actor|SetLifeSpan 0.01))))
    (:Is%20Not%20Valid
      _)))

(event EventDestroyed
  (bind _weaponactor (Variables|Default|GetWeaponActor))
  (bind _rope (Variables|RopeTether|GetRope))
  ;[1 0.3879 0] Handle Destroy
  (Variables|Hit|SetHit%20Actor 0)
  (Utilities|IsValid _weaponactor
    (:Is%20Valid
      (Variables|/TowHookWeapon/Gameplay/B_TowHookWeapon_Athena.B_TowHookWeapon_Athena_C|SetProjectileHitData_Projectile_9_FF96757148C1260DFF8E428A8DB9D153 (struct :Projectile_9_FF96757148C1260DFF8E428A8DB9D153 0 :HitActor_7_424357AE4112F32CB1760FB82D90F298 0 :LocalHitLocation_10_41BBF2064BDE9C04D5F1D9A7D835FD6E "0.000000,0.000000,0.000000") _weaponactor))
    (:Is%20Not%20Valid
      _))
  (Utilities|IsValid _rope
    (:Is%20Valid
      (Variables|/Script/FortniteGame.FortHookGunRope|SetbProjectileGone true _rope)
      (Development|PrintString "Send End Reel"))
    (:Is%20Not%20Valid
      _))
  (Niagara|SpawnSystematLocation (Variables|FX|GetFX_Break) (Transformation|GetActorLocation) (Math|Conversions|RotationFromXVector (Variables|Hit|GetHit%20Normal)))
  ;[0 0.1341 0.5521] Audio — remove projectile from vehicle sound component
  (bind _vehicle_sound_component (Variables|Default|GetVehicle%20Sound%20Component)
    (:then
      (Class|BTowHookVehicleSoundComponent|ModifyAttachCount _vehicle_sound_component -1))
    (:else)))

(event Custom|DetachRopeFromProjectile
  (bind _rope (Variables|RopeTether|GetRope))
  ;[0 0.05344 1] Detach the projectile from the rope.  Retract Rope. Play Cue
  (Utilities|IsValid _rope
    (:Is%20Valid
      (bind _asb_athena_tow_hook_weapon_wire (Utilities|Casting|CastToB_Athena_TowHookWeapon_Wire _rope))
      (Class|BAthenaTowHookWeaponWire|EndReel _asb_athena_tow_hook_weapon_wire)
      ;;[0.3909 0 1] Line Snap Gameplay Cue
      (GameplayCue|ExecuteGameplayCue (Variables|Default|GetPlayerPawn) (Variables|Tags|GetLine%20Snap%20Gameplay%20Cue%20Tag)))))

(event Custom|KillProjectile
  (bind _rope (Variables|RopeTether|GetRope))
  ;[0 1 0.25] Kill Projectile. This can be called from this BP, or from the GA_Athena_TowHookWeapon_Tethered_Passive that watches for disconnects
  (Utilities|Time|ClearandInvalidateTimerbyHandle (Variables|RopeTether|GetPrjTimer)) ; clean up timers on attach
  (Utilities|Time|ClearandInvalidateTimerbyHandle (Variables|Default|GetFailsafeTimer))
  (Utilities|IsValid _rope
    (:Is%20Valid
      (Variables|/Script/FortniteGame.FortHookGunRope|SetbProjectileGone true _rope))
    (:Is%20Not%20Valid
      _))
  (Rendering|SetVisibility (Variables|Default|GetStaticMesh) false false)
  ;;[0 0.1094 1] FORT-734280 - Speculative fix. Let projectile live just a little bit longer so the rope has a valid point to attach to
  (Actor|SetLifeSpan (ScalableFloat|GetValueatLevel (Variables|ScalableFloats|GetSF_ProjectileLifetimeOnKillPrj))))

(event EventBeginPlay
  (bind _ns_coresniper_bullet_trail (Variables|Default|GetNS_CoreSniper_Bullet_Trail))
  (bind _ns_coresniper_bullet_trail_2 (Variables|Default|GetNS_CoreSniper_Bullet_Trail))
  ;[1 1 1] Spawn and attach wire to projectile. Wire for the actual tether is handled in the weapon BP OnRepProjectileHitData notify
  (CallFunction|SetHotfixValuestoVars)
  (bind _asfort_player_pawn (Utilities|Casting|CastToFortPlayerPawn (Game|GetInstigator)))
  (bind _output_get (Variables|Default|SetFortPawn _asfort_player_pawn))
  (Ability|SendGameplayEventtoActor _output_get (Variables|Tags|GetEvent_ProjSpawned))
  (bind _asb_tow_hook_weapon_athena (Utilities|Casting|CastToB_TowHookWeapon_Athena (Actor|GetOwner)))
  (Variables|Default|SetWeaponActor _asb_tow_hook_weapon_athena)
  (if (Networking|IsServer)
    (bind _returnvalue (Game|SpawnActorfromClass "/TowHookWeapon/Gameplay/B_Athena_TowHookWeapon_Wire.B_Athena_TowHookWeapon_Wire_C" (Math|Transform|MakeTransform (Transformation|GetActorLocation)) "Undefined" "MultiplyWithRoot" (Variables|Default|GetWeaponActor) (Variables|Default|GetFortPawn) 0))
    (bind _output_get_2 (Variables|RopeTether|SetRope _returnvalue))
    (Class|FortHookGunRope|SetWeaponandProjectile _output_get_2 (Variables|Default|GetWeaponActor) self)
    (Variables|/TowHookWeapon/Gameplay/B_TowHookWeapon_Athena.B_TowHookWeapon_Athena_C|SetRope (Variables|RopeTether|GetRope) (Variables|Default|GetWeaponActor))
    ;[0 0.05344 1] This timer determines when to detach the projectile from the rope. Retract rope
    (bind _returnvalue_2 (Utilities|Time|SetTimerbyEvent (EventDispatchers|CreateEvent :!event Custom|DetachRopeFromProjectile) (ScalableFloat|GetValueatLevel (Variables|ScalableFloats|GetSF_RopeDetachTimer))))
    (Variables|RopeTether|SetPrjTimer _returnvalue_2)
    ;[1 0.3879 0] Failsafe timeout in the event something unexpected happens
    (bind _returnvalue_3 (Utilities|Time|SetTimerbyEvent (EventDispatchers|CreateEvent :!event Custom|AdjustPrjLifetime) 5.0))
    (Variables|Default|SetFailsafeTimer _returnvalue_3)
    (else
      (Utilities|FlowControl|Sequence
        (:then_0
          ;[0.3348 0 1] Trail
          ;;[1 1 1] Hacky fix for FORT-587865
          (if (Fortnite|IsMobilePlatform)
            (Components|Activation|Deactivate _ns_coresniper_bullet_trail_2)
            (Utilities|FlowControl|Delay 0.04)
            (Components|Activation|Activate _ns_coresniper_bullet_trail_2 false)
            (else
              (Niagara|SetNiagaraVariable(Vector3) _ns_coresniper_bullet_trail "BulletOrigin" (Effects|GetMuzzleLocation (Variables|/Script/FortniteGame.FortPawn|GetCurrentWeapon (Variables|Default|GetFortPawn)) -1))
              (Niagara|SetNiagaraVariable(Float) _ns_coresniper_bullet_trail "StartTime" (Utilities|Time|GetGameTimeinSeconds))
              (Niagara|SetNiagaraVariable(Float) _ns_coresniper_bullet_trail "WPODuration" 0.2))))
        (:then_1
          (Utilities|FlowControl|Sequence
            (:then_0
              (CallFunction|PlayBulletWhip))))))))

(event Custom|StopLoopingNSTrail
  ;[0.3348 0 1] Kill VFX & SFX trail
  (Networking|CanExecuteCosmeticEvents
    (:True
      (Transformation|DetachFromComponent (Variables|Default|GetNS_CoreSniper_Bullet_Trail) "KeepWorld" "KeepWorld" "KeepWorld" true)
      (Utilities|FlowControl|Sequence
        (:then_0
          (Niagara|SetNiagaraVariable(Bool) (Variables|Default|GetNS_CoreSniper_Bullet_Trail) "Deactive" true)
          (Utilities|FlowControl|Delay 0.2)
          (Components|Activation|Deactivate (Variables|Default|GetNS_CoreSniper_Bullet_Trail)))))))

(event Custom|StopLoopingAudioTell
  ;[0.3348 0 1] Kill VFX & SFX trail
  (Audio|Components|Audio|FadeOut (Variables|Default|GetLoopingAudioTell) 0.5 0.0 "Linear" :!class /Script/Engine.AudioComponent))

(event Custom|CheckValidTetherActorHits
  (bind _hit_actor (Variables|Hit|GetHit%20Actor))
  (bind _returnvalue (Actor|GetAttachParentActor _hit_actor))
  (bind _rope (Variables|RopeTether|GetRope))
  (bind _rope_2 (Variables|RopeTether|GetRope))
  (bind (_projectilehitdata_projectile_9_ff96757148c1260dff8e428a8db9d153 _projectilehitdata_hitactor_7_424357ae4112f32cb1760fb82d90f298) (Utilities|Struct|BreakTowHookProjectileHit (Variables|/TowHookWeapon/Gameplay/B_TowHookWeapon_Athena.B_TowHookWeapon_Athena_C|GetProjectileHitData_Projectile_9_FF96757148C1260DFF8E428A8DB9D153 (Variables|Default|GetWeaponActor))))
  ;[0.01173 1 0.0532] [ATTACH] If we're a valid vehicle or train, attach and tether player, otherwise kill the projectile.
  (Networking|SwitchHasAuthority
    (:Authority
      (if (<= (+ (Math|Vector|DistanceSquared(Vector) (Transformation|GetActorLocation (Variables|Default|GetFortPawn)) (Variables|Hit|GetHit%20Location)) (ScalableFloat|GetValueatLevel (Variables|ScalableFloats|GetSF_ExtraAttachAllowanceDistance))) (+ (Math|Float|Square (ScalableFloat|GetValueatLevel (Variables|ScalableFloats|GetSF_AttachDistance))) (ScalableFloat|GetValueatLevel (Variables|ScalableFloats|GetSF_ExtraAttachAllowanceDistance))))
        (Utilities|Casting|CastToFortVehicleInterface _hit_actor
          (:then
            _)
          (:CastFailed
            (Utilities|Casting|CastToFortVehicleInterface _returnvalue
              (:then
                (Variables|Hit|SetHit%20Actor _returnvalue))
              (:CastFailed
                (bind _returnvalue_2 (GameplayTags|HasMatchingGameplayTag(Message) (Variables|Hit|GetHit%20Actor) (Variables|Tags|GetTrain%20Tag)))
                (if _returnvalue_2
                  (Utilities|IsValid _rope
                    (:Is%20Valid
                      (Class|BAthenaTowHookWeaponWire|TightenRope _rope)
                      ;;[0.3042 0 1] Vehicle Impact Gameplay Cue
                      (GameplayCue|ExecuteGameplayCue (Variables|Hit|GetHit%20Actor) (Variables|Tags|GetVehicle%20Hit%20Gameplay%20Cue%20Tag))
                      ;[0 0.9492 1] If same actor is hit, just damage, don't reattach etc.
                      (if (== (Variables|Hit|GetHit%20Actor) _projectilehitdata_hitactor_7_424357ae4112f32cb1760fb82d90f298)
                        (CallFunction|KillProjectile)
                        (else
                          (CallFunction|AttachProjectiletoValidActors))))
                    (:Is%20Not%20Valid
                      _))
                  (else
                    (Utilities|IsValid _rope_2
                      (:Is%20Valid
                        (Class|BAthenaTowHookWeaponWire|EndReel _rope_2)
                        ;;[0.3042 0 1] Line Snap Gameplay Cue
                        (GameplayCue|ExecuteGameplayCue (Variables|Default|GetPlayerPawn) (Variables|Tags|GetLine%20Snap%20Gameplay%20Cue%20Tag))
                        (CallFunction|KillProjectile))
                      (:Is%20Not%20Valid
                        _))))
                (Development|PrintString "ROPE NOT VALID WHEN IT SHOULD BE" true true "(R=1.000000,G=0.000000,B=0.018539,A=1.000000)" 3.0)))))
        (else
          (CallFunction|KillProjectile))))))

(event Collision|EventHit
  (params (MyComp :type Primitive%20Component%20Object%20Reference) (Other :type Actor%20Object%20Reference) (OtherComp :type Primitive%20Component%20Object%20Reference) (bSelfMoved :type Boolean) (HitLocation :type Vector :default "0, 0, 0") (HitNormal :type Vector :default "0, 0, 0") (NormalImpulse :type Vector :default "0, 0, 0") (Hit :type Hit%20Result%20Structure%20%28by%20ref%29))
  ;[0 0.9052 1] Handle Hits
  (Variables|Hit|SetHit%20Location HitLocation)
  (Variables|Hit|SetHit%20Normal HitNormal)
  (Variables|Hit|SetHit Hit)
  (Variables|Hit|SetHit%20Actor Other)
  (Collision|SetActorEnableCollision false)
  (Utilities|IsValid (Variables|RopeTether|GetRope)
    (:Is%20Valid
      (CallFunction|SendQuestEvent Other)
      (Utilities|FlowControl|Sequence
        (:then_0
          (CallFunction|StopLoopingNSTrail)
          ;;[0.5278 0 1] Audio events on hit
          (bind _weaponactor (Variables|Default|GetWeaponActor)
            (:then
              (Class|BTowHookWeaponAthena|OnProjectileHit _weaponactor (Variables|Hit|GetHit%20Actor) self)
              (CallFunction|StopLoopingAudioTell))
            (:else)))
        (:then_1
          (CallFunction|CheckValidTetherActorHits))))))

(event Custom|AttachProjectileToValidActors
  (bind _rope (Variables|RopeTether|GetRope))
  ;[0 0.5976 1] Handle Projectile Attach, Visuals etc.
  ;;[0 0.0959 1] Eject player if hitting a vehicle while you are a passenger.
  (Utilities|Casting|CastToFortAthenaVehicle (Variables|Hit|GetHit%20Actor)
    (:then
      (bind _returnvalue (CallFunction|ShouldPlayerbeEjectedfromCurrentVehicle))
      (if _returnvalue
        (bind _returnvalue_2 (Vehicle|EjectPlayerPawn(Message) (Vehicle|GetVehicle (Variables|Default|GetFortPawn)) (Variables|Default|GetFortPawn)))
        (if _returnvalue_2
          ;;;[1 0.9358 0] TODO: hacky solution for exiting current vehicle to attach to new
          (Utilities|FlowControl|Delay (ScalableFloat|GetValueatLevel (Variables|ScalableFloats|GetSF_DelayAfterEjectFromVehicle)))
          (else
            _))
        (Utilities|FlowControl|Sequence
          (:then_0
            ;;[1 0.5313 0] Tow Hook Func Library:
            ;; Break and override hit loc to attach prj to vehicle
            (Default|AttachPrjtoVehicleServer (Variables|Hit|GetHit%20Actor) (Variables|Hit|GetHit) self)
            (Variables|/TowHookWeapon/Gameplay/B_TowHookWeapon_Athena.B_TowHookWeapon_Athena_C|SetRope (Variables|RopeTether|GetRope) (Variables|Default|GetWeaponActor))
            (Class|BTowHookWeaponAthena|UpdateHUDKeys (Variables|Default|GetWeaponActor) true)
            ;;[0.9427 1 0] TODO: HACK for getting client the Hit Actor to start the Movement Mode Extension
            (Variables|/TowHookWeapon/Gameplay/B_TowHookWeapon_Athena.B_TowHookWeapon_Athena_C|SetProjectileHitData_Projectile_9_FF96757148C1260DFF8E428A8DB9D153 (Utilities|Struct|MakeTowHookProjectileHit self (Variables|Hit|GetHit%20Actor) (Math|Transform|InverseTransformLocation (Transformation|GetActorTransform (Variables|Hit|GetHit%20Actor)) (Variables|Hit|GetHit%20Location))) (Variables|Default|GetWeaponActor))
            ;;[0 0.1341 0.5521] Audio — if a vehicle was hit, add a vehicle sound component to it or use one that already exists
            (bind _hit_actor (Variables|Hit|GetHit%20Actor)
              (:then
                (Utilities|IsValid (Actor|GetComponentbyClass _hit_actor "/TowHookWeapon/Assets/Audio/B_TowHookVehicleSoundComponent.B_TowHookVehicleSoundComponent_C")
                  (:Is%20Valid
                    (Variables|Default|SetVehicle%20Sound%20Component (Actor|GetComponentbyClass _hit_actor "/TowHookWeapon/Assets/Audio/B_TowHookVehicleSoundComponent.B_TowHookVehicleSoundComponent_C")))
                  (:Is%20Not%20Valid
                    (bind _returnvalue_3 (Game|AddComponentbyClass _hit_actor "/TowHookWeapon/Assets/Audio/B_TowHookVehicleSoundComponent.B_TowHookVehicleSoundComponent_C"))
                    (Variables|Default|SetVehicle%20Sound%20Component _returnvalue_3)))
                (Class|BTowHookVehicleSoundComponent|ModifyAttachCount (Variables|Default|GetVehicle%20Sound%20Component) 1)
                ;;[0.9427 1 0] HACK: Fix post 30.10
                ;; This magic wait is to sync up visuals on clients for hook attach point
                ;; bool Rep Notify is to hide the static mesh so it doesn't sit there floating in air while this wait happens
                (if (ScalableFloat|GetValueasBool (Variables|ScalableFloats|GetSF_HackDelayEnabled))
                  (Variables|Default|SetbRepNotifyToHideStaticMesh true)
                  (Utilities|FlowControl|Delay (ScalableFloat|GetValueatLevel (Variables|ScalableFloats|GetSF_HackDelayTimerForVisualAttachToSocket)))
                  ;;[1 0 0.9324] RepNotify to clients to move visuals ( mesh ) to correct socket location on cosmetic vehicles.
                  ;; ( Cosmetic vehicles sockets are always Octane locations on server )
                  (Variables|Hit|SetHit%20Actor_RepNotifyForClientMeshLoc (Variables|Hit|GetHit%20Actor))))
              (:else)))
          (:then_1
            (Actor|SetLifeSpan 0.0)
            ;;[1 0.3879 0] FORT-753580 - if tether ends due to Block MME tags in TetherExtLogic, we never hit any events there, so we the destroy actor call to the proj never happens, and we wind up here.
            ;; Determine if we're not attached to anything and kill projectile
            (if (Variables|Default|GetbUseFailsafeTimerAfterAttach)
              (Utilities|Time|ClearandInvalidateTimerbyHandle (Variables|Default|GetFailsafeTimer))
              (bind _returnvalue_4 (Utilities|Time|SetTimerbyEvent (EventDispatchers|CreateEvent :!event Custom|AdjustPrjLifetime) (Variables|Default|GetFailSafeTimerFloat)))
              (Variables|Default|SetFailsafeTimer _returnvalue_4)))
          (:then_2
            (Utilities|IsValid _rope
              (:Is%20Valid
                (Actor|DestroyActor _rope)))))
        (else
          _)))
    (:CastFailed
      _)))

(event OnEnterWaterEvent(WaterInteractionComponent)
  (params (WaterBody :type Fort%20Water%20Body%20Actor%20Object%20Reference) (WaterInteractionComponent :type Fort%20Water%20Interaction%20Component%20Object%20Reference) (bIsFirstBody :type Boolean))
  (bind _returnvalue (Transformation|GetActorLocation))
  (bind _weaponactor (Variables|Default|GetWeaponActor))
  (bind _returnvalue_2 (Transformation|GetActorLocation _weaponactor))
  ;[0 0.298 1] Water Impact FX - Fixed Impact -> Ask Philippe Patient for any concerns
  (Networking|CanExecuteCosmeticEvents
    (:True
      (if bIsFirstBody
        (Utilities|IsValid WaterBody
          (:Is%20Valid
            (bind (_outhit _returnvalue_3) (Collision|LineTraceForObjects _returnvalue_2 _returnvalue (Utilities|Array|MakeArray "ObjectTypeQuery13")))
            (Rendering|Debug|DrawDebugLine _returnvalue_2 _returnvalue)
            (bind _outwatersurfaceinfo (WaterBody|GetWaterSurfaceInfoStruct WaterBody (select _returnvalue_3 (.ImpactPoint (Collision|BreakHitResult _outhit)) (Transformation|GetActorLocation)) true true false))
            (Audio|PlaySoundatLocation (Variables|Audio|GetImpact%20Water%20Sound) (Utilities|Struct|BreakWaterSurfaceInfo _outwatersurfaceinfo))
            (bind _returnvalue_4 (Effects|Components|ParticleSystem|SpawnFXAtLocation (Variables|FX|GetWaterImpactFX) (Utilities|Struct|BreakWaterSurfaceInfo _outwatersurfaceinfo) (Math|Rotator|MakeRotfromX (.WaterPlaneNormal (Utilities|Struct|BreakWaterSurfaceInfo _outwatersurfaceinfo)))))
            (Utilities|IsValid _returnvalue_4
              (:Is%20Valid
                (Effects|Components|ParticleSystem|SetFloatParameter _returnvalue_4 "WaterDepth" (.WaterDepth (Utilities|Struct|BreakWaterSurfaceInfo _outwatersurfaceinfo)))
                (Effects|Components|ParticleSystem|SetFloatParameter _returnvalue_4 "WaterBodyIndex" (Math|Conversions|ToFloat(Integer) (.WaterBodyIdx (Utilities|Struct|BreakWaterSurfaceInfo _outwatersurfaceinfo))))))))))))

(fn OnRep_bRepNotifyToHideStaticMesh
  (Rendering|SetVisibility (Variables|Default|GetStaticMesh) false false))

(fn OnRep_HitActor_RepNotifyForClientMeshLoc
  (bind _attachvisualhookstaticmesh (Variables|Default|GetAttachVisualHookStaticMesh))
  (Networking|SwitchHasAuthority
    (:Remote
      (bind _asfort_athena_vehicle (Utilities|Casting|CastToFortAthenaVehicle (Variables|Hit|GetHit%20Actor_RepNotifyForClientMeshLoc)))
      (if (Transformation|DoesSocketExist (Variables|/Script/FortniteGame.FortAthenaVehicle|GetSkeletalMesh _asfort_athena_vehicle) "TowHookAttach")
        (Rendering|SetVisibility (Variables|Default|GetStaticMesh) false false)
        (bind (_returnvalue_location _returnvalue_rotation) (Math|Transform|BreakTransform (Transformation|GetSocketTransform (Variables|/Script/FortniteGame.FortAthenaVehicle|GetSkeletalMesh _asfort_athena_vehicle) "TowHookAttach" "RTS_World")))
        (Transformation|SetWorldLocationAndRotation _attachvisualhookstaticmesh _returnvalue_location _returnvalue_rotation false true)
        (Rendering|SetVisibility _attachvisualhookstaticmesh true false)))))

(fn PlayBulletWhip
  (if (not (Pawn|IsLocallyViewed (Variables|Actor|GetInstigator)))
    ;[0.06578 0.07735 0.1302] Play the bullet whip if the player is within a certain range (not too close to the player, but close enough that the projectile doesn't have time to update the metasound with its position before it is out of range, causing the whip to not play). Also checks if the instigator is facing the player
    (bind _asfort_player_pawn (Utilities|Casting|CastToFortPlayerPawn (Variables|Actor|GetInstigator)))
    (if (and (Math|Float|InRange(Float) (Transformation|GetDistanceTo (Variables|Actor|GetInstigator) (Game|GetPlayerPawn 0)) 2e+03 7.2e+03) (Math|Float|InRange(Float) (Math|Trig|RadiansToDegrees (Math|Trig|Acos(Radians) (Math|Vector|DotProduct (Transformation|GetForwardVector (Variables|/Script/Engine.Actor|GetRootComponent _asfort_player_pawn)) (Math|Vector|Normalize (- (Transformation|GetActorLocation _asfort_player_pawn) (Transformation|GetActorLocation (Game|GetPlayerPawn 0))))))) (- 1.8e+02 (LocalVariables|GetAngleRange)) (+ (LocalVariables|GetAngleRange) 1.8e+02)))
      (Audio|Parameter|ExecuteTriggerParameter (Variables|Default|GetLoopingAudioTell) "Events.Fire"))))

(fn SendQuestEvent
  (params (Hit%20Actor :type Actor%20Object%20Reference))
  (Ability|SendGameplayEventtoActor (Variables|Default|GetFortPawn) "(TagName=\"Event.Athena.Item.TowHookWeapon.Hit\")" (Utilities|Struct|MakeGameplayEventData :Instigator (Variables|Default|GetFortPawn) :Target Hit%20Actor :OptionalObject self :ContextHandle "()" :InstigatorTags "(GameplayTags=)" :TargetTags "(GameplayTags=)" :EventMagnitude 0.0 :TargetData "()")))

(fn SetHotfixValuesToVars
  (bind _output_get (Variables|Default|SetFailSafeTimerFloat (ScalableFloat|GetValueatLevel (Variables|ScalableFloats|GetSF_FailsafeTimerAfterAttach))))
  (Variables|Default|SetbUseFailsafeTimerAfterAttach (!= _output_get 0)))

(fn ShouldPlayerBeEjectedFromCurrentVehicle
  (bind _returnvalue (Vehicle|GetVehicle (Variables|Default|GetFortPawn)))
  (bind _fortpawn (Variables|Default|GetFortPawn))
  ;[0 1 0.7882] Any seat will eject if motorcycle; otherwise passenger only as only passengers can fire
  (if (ScalableFloat|GetValueasBool (Variables|ScalableFloats|GetSF_EnableEjectFromVehicle))
    (if (Vehicle|IsDrivingVehicle _fortpawn)
      (Utilities|IsValid _returnvalue
        (:Is%20Valid
          (return (GameplayTags|HasMatchingGameplayTag _returnvalue "(TagName=\"Athena.Vehicle.Motorcycle\")")))
        (:Is%20Not%20Valid
          _))
      (else
        _))
    (return (Vehicle|IsPassengerinVehicle (Variables|Default|GetFortPawn)))
    (else
      (return false))))

(fn ConstructionScript
  (|CallParentFunction))
