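; Movement mode extension logic for the tow hook weapon's tethered state, managing the pawn's animation, camera, HUD and effects.
; Push and end link and unlink a client anim layer, relinked after customization; end also destroys the projectile and grants mounting vehicle immunity.
; Tether begin and end toggle camera, reel-in prompt, GameplayEffects and a delayed immunity-removal timer.
; Grounded, water and surface events drive force feedback and skid effects, with a hotfix fallback to one combined effect.
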(event EventOnExtensionActivate
  (params (RuntimeContext :type MMERuntime%20Context%20Structure%20%28by%20ref%29))
  (Class|FortMovementModeExtBlueprintLibrary|GetRuntimeDatafromContext RuntimeContext "/Script/TowHookWeaponRuntime.FortMovementMode_TetheredRuntimeData")
  (Variables|Default|SetUseCombinedSkiddingNS (ScalableFloat|GetValueasBool (Variables|Default|GetHF_UseCombinedNS_GCNL))))

(event EventOnExtensionPush
  (params (RuntimeContext :type MMERuntime%20Context%20Structure%20%28by%20ref%29))
  (bind _returnvalue (Class|FortMovementModeExtBlueprintLibrary|GetRuntimeDatafromContext RuntimeContext "/Script/TowHookWeaponRuntime.FortMovementMode_TetheredRuntimeData"))
  (if (not (CallFunction|IsAnotherTetheredMMEActive _returnvalue))
    (CallFunction|LinkAnimLayer)
    ;[1 1 1] The anim layer will be unlinked after customization, so reapply it back if this happens when tethered (possible e.g. when a tethered character enters relevancy).
    (bind _asfort_player_pawn (Utilities|Casting|CastToFortPlayerPawn (Variables|FortMovementMode_BaseExtLogic|GetOwnerPawn)))
    (Customization|BindEventtoOnFinishedCharacterCustomization _asfort_player_pawn (delegate Custom|OnFinishedCustomisation))))

(event Custom|OnFinishedCustomisation
  (params (Pawn :type Fort%20Player%20Pawn%20Object%20Reference))
  ;[1 1 1] The anim layer will be unlinked after customization, so reapply it back if this happens when tethered (possible e.g. when a tethered character enters relevancy).
  (CallFunction|LinkAnimLayer))

(event EventOnExtensionEnd
  (params (RuntimeContext :type MMERuntime%20Context%20Structure%20%28by%20ref%29))
  (bind _returnvalue (Class|FortMovementModeExtBlueprintLibrary|GetRuntimeDatafromContext RuntimeContext "/Script/TowHookWeaponRuntime.FortMovementMode_TetheredRuntimeData"))
  (Utilities|IsValid (Variables|/Script/TowHookWeaponRuntime.FortMovementMode_TetheredRuntimeData|GetNetProjectile _returnvalue)
    (:Is%20Valid
      (Actor|DestroyActor (Variables|/Script/TowHookWeaponRuntime.FortMovementMode_TetheredRuntimeData|GetNetProjectile _returnvalue)))
    (:Is%20Not%20Valid
      _))
  (if (not (CallFunction|IsAnotherTetheredMMEActive _returnvalue))
    (CallFunction|UnlinkAnimLayer)
    ;[1 1 1] Unbind customization event
    (bind _asfort_player_pawn (Utilities|Casting|CastToFortPlayerPawn (Variables|FortMovementMode_BaseExtLogic|GetOwnerPawn))
      (:then
        (Customization|UnbindEventfromOnFinishedCharacterCustomization _asfort_player_pawn (delegate Custom|OnFinishedCustomisation)))
      (:CastFailed
        _))
    (CallFunction|RemoveNonMMEPushedTetheredTag)
    ;[0 1 0.1327] Vehicle Immunity Application
    (CallFunction|ApplyVehicleImmuneTags RuntimeContext)))

(event EventHandleTetheredMovementBegun
  (bind _ownerpawn (Variables|FortMovementMode_BaseExtLogic|GetOwnerPawn))
  (bind _returnvalue (Actor|GetComponentbyClass _ownerpawn "/Script/TowHookWeaponRuntime.FortPawnComponent_TetheredMovement"))
  (Variables|Default|SetTetheredPawnComponent (Actor|GetComponentbyClass (Variables|FortMovementMode_BaseExtLogic|GetOwnerPawn) "/Script/TowHookWeaponRuntime.FortPawnComponent_TetheredMovement"))
  (Utilities|FlowControl|Sequence
    (:then_0
      ;[0 0.05721 1] Camera Modfier Apply
      (bind _returnvalue_2 (Fortnite|Camera|AddCameraModifierforOwningPlayerController (Variables|Camera|GetCameraModifier) (Variables|FortMovementMode_BaseExtLogic|GetOwnerPawn)))
      (Variables|Camera|SetCameraModifierObjectRef _returnvalue_2)
      ;[0.3337 0 1] Apply GE for tag outside of Push tags, which are moved when this MME is deactivated.
      ; When tag Granted.Athena.Item.TowHookWeapon.PlayerTetheredAnyState is removed, we kill the projectile
      (CallFunction|AddNonMMEPushedTetheredTag)
      ;[0 0.3807 1] Show Hud Reel In Input based on tether state
      (Utilities|Time|SetTimerbyEvent (EventDispatchers|CreateEvent :!event Custom|Show%20Reel%20In%20HUD%20Key%20Text) 0.1 false true)
      ;[0 0.1438 1] Bind Grounded & Water state events. Used for Force Feedback & GCNLs
      (EventDispatchers|BindEventtoOnGroundedStateChanged _returnvalue (EventDispatchers|CreateEvent :!event Custom|OnGroundedStateChangedCustomEvent))
      (EventDispatchers|BindEventtoOnWaterStateChanged _returnvalue (EventDispatchers|CreateEvent :!event Custom|OnWaterStateChangeCustomEvent)))
    (:then_1
      (Default|ApplyMMEGEANDSpeedLineGCNL)
      ;[0.6633 0 1] Hotfix to determine which GCNL we want to use
      ; We are tight on time, so if we have too many complications, use the working combined, but lless perf friendly version
      (if (Variables|Default|GetUseCombinedSkiddingNS)
        (Default|ApplyCombinedNSSkiddingGCNL)
        (else
          (EventDispatchers|BindEventtoOnSurfaceTypeChanged (Footsteps|GetSurfaceTrackingComponent (Variables|FortMovementMode_BaseExtLogic|GetOwnerPawn)) (EventDispatchers|CreateEvent :!event Custom|OnSurfaceChanged)))))))

(event EventHandleTetheredMovementEnded
  (bind _ownerpawn (Variables|FortMovementMode_BaseExtLogic|GetOwnerPawn))
  (bind _returnvalue (Actor|GetComponentbyClass _ownerpawn "/Script/TowHookWeaponRuntime.FortPawnComponent_TetheredMovement"))
  (bind _cameramodifierobjectref (Variables|Camera|GetCameraModifierObjectRef))
  (Utilities|FlowControl|Sequence
    (:then_0
      ;[0 0.05721 1] Remove Camera Modifier
      (Utilities|IsValid _cameramodifierobjectref
        (:Is%20Valid
          (bind _astow_hook_camera_modifier (Utilities|Casting|CastToTowHookCameraModifier _cameramodifierobjectref))
          (Class|TowHookCameraModifier|BlendOut _astow_hook_camera_modifier)
          ;[1 1 1] Line Snap Gameplay Cue
          (GameplayCue|ExecuteGameplayCue (Variables|FortMovementMode_BaseExtLogic|GetOwnerPawn) (Variables|GameplayTags|GetLine%20Snap%20Gameplay%20Cue%20Tag))
          ;[0 0.3807 1] Hide Hud Reel In Input based on tether state
          (Utilities|Time|SetTimerbyEvent (EventDispatchers|CreateEvent :!event Custom|Hide%20Reel%20In%20HUD%20Key%20Text) 0.1 false true)
          ;[0 0.1438 1] Unbind Grounded & Water state events. Stop Force Feedback
          (EventDispatchers|UnbindallEventsfromOnGroundedStateChanged _returnvalue)
          (EventDispatchers|UnbindallEventsfromOnWaterStateChanged _returnvalue)
          (Game|Feedback|LocalStopForceFeedback (Fortnite|GetFortPlayerControllerfromActor (Variables|FortMovementMode_BaseExtLogic|GetOwnerPawn)) (Variables|FF|GetLooping%20Force%20Feedback%20Effect) "None")
          (if (not (Variables|Default|GetUseCombinedSkiddingNS))
            (EventDispatchers|UnbindEventfromOnSurfaceTypeChanged (Footsteps|GetSurfaceTrackingComponent (Variables|FortMovementMode_BaseExtLogic|GetOwnerPawn)) (EventDispatchers|CreateEvent :!event Custom|OnSurfaceChanged))
            (CallFunction|RemoveWaterSkiingGCNL)
            (CallFunction|RemoveSkiddingGCNL)))))
    (:then_1
      (Default|RemoveMMEGEsANDGCNLs)
      (Variables|VehicleImmunity|SetImmunityRemovalTimeCheck (ScalableFloat|GetValueatLevel (Variables|VehicleImmunity|GetHF_VehicleImmunityRemovalDelay)))
      ;[0 1 0.1327] Vehicle Immunity Removal
      (bind _returnvalue_2 (Utilities|Time|SetTimerbyEvent (EventDispatchers|CreateEvent :!event Custom|RemoveVehicleImmuneTags) (Variables|VehicleImmunity|GetImmunityRemovalTimeCheck) true true))
      (Variables|VehicleImmunity|SetVehicleImmuneRemovalTimerHandle _returnvalue_2)
      ;[0.6633 0 1] Hotfix to determine which GCNL we want to use
      ; We are tight on time, so if we have too many complications, use the working combined, but lless perf friendly version
      (if (Variables|Default|GetUseCombinedSkiddingNS)
        (CallFunction|RemoveCombinedNSSkiddingGCNL)))))

(event Custom|RemoveVehicleImmuneTags
  (bind _returnvalue (CallFunction|GetMovementRuntimeData))
  (bind _vehicleimmuneremovaltimerhandle (Variables|VehicleImmunity|GetVehicleImmuneRemovalTimerHandle))
  ;[0 1 0.1327] Vehicle Immunity Events
  (Utilities|IsValid _returnvalue
    (:Is%20Valid
      (if (not (Variables|/Script/TowHookWeaponRuntime.FortMovementMode_TetheredRuntimeData|GetbMountingTarget _returnvalue))
        (GameplayEffects|RemoveActiveEffectswithGrantedTags (Ability|GetFortAbilitySystemComponent (Variables|FortMovementMode_BaseExtLogic|GetOwnerPawn)) (Variables|GameplayTags|GetVehicleImmuneTags))
        (if (Utilities|Time|IsValidTimerHandle _vehicleimmuneremovaltimerhandle)
          (Utilities|Time|InvalidateTimerHandle _vehicleimmuneremovaltimerhandle))))))

(event Custom|ApplyVehicleImmuneTags
  (params (RuntimeContext :type MMERuntime%20Context%20Structure))
  ;[0 1 0.1327] Vehicle Immunity Events
  (bind _returnvalue (Class|FortMovementModeExtBlueprintLibrary|GetRuntimeDatafromContext RuntimeContext "/Script/TowHookWeaponRuntime.FortMovementMode_TetheredRuntimeData"))
  (Utilities|IsValid _returnvalue
    (:Is%20Valid
      (if (Variables|/Script/TowHookWeaponRuntime.FortMovementMode_TetheredRuntimeData|GetbMountingTarget _returnvalue)
        (GameplayEffects|ApplyGameplayEffectToSelf :self (Ability|GetAbilitySystemComponent (Variables|FortMovementMode_BaseExtLogic|GetOwnerPawn)) :GameplayEffectClass (Variables|GameplayEffects|GetGE_VehicleImmuneTags) :Level 1.0)
        (else
          (Development|PrintString 0 "Tether Runtime Data - Mounting Target False when it should be True"))))
    (:Is%20Not%20Valid
      (Development|PrintString 0 "Tether Runtime Data - Not valid when it should be"))))

(event Custom|Hide%20Reel%20In%20HUD%20Key%20Text
  ;[0 0.3807 1] HUD input text adjustment
  (bind _asb_tow_hook_weapon_athena (Utilities|Casting|CastToB_TowHookWeapon_Athena (Variables|/Script/FortniteGame.FortPawn|GetCurrentWeapon (Variables|FortMovementMode_BaseExtLogic|GetOwnerPawn))))
  (WeaponHUDData|SetHudKeyActionVisibility _asb_tow_hook_weapon_athena 5 false))

(event Custom|Show%20Reel%20In%20HUD%20Key%20Text
  ;[0 0.3807 1] HUD input text adjustment
  (bind _asb_tow_hook_weapon_athena (Utilities|Casting|CastToB_TowHookWeapon_Athena (Variables|/Script/FortniteGame.FortPawn|GetCurrentWeapon (Variables|FortMovementMode_BaseExtLogic|GetOwnerPawn))))
  (WeaponHUDData|SetHudKeyActionVisibility _asb_tow_hook_weapon_athena 5 true))

(event Custom|OnGroundedStateChangedCustomEvent
  (params (bGrounded :type Boolean))
  ;[0 0.1438 1] Start & Stop Force Feedback & Skidding GCNLs based on state
  (if bGrounded
    (Game|Feedback|ClientPlayForceFeedback (Fortnite|GetFortPlayerControllerfromActor (Variables|FortMovementMode_BaseExtLogic|GetOwnerPawn)) (Variables|FF|GetLooping%20Force%20Feedback%20Effect) "None" true false false)
    (Game|Feedback|ClientPlayForceFeedback (Fortnite|GetFortPlayerControllerfromActor (Variables|FortMovementMode_BaseExtLogic|GetOwnerPawn)) (Variables|FF|GetOnLanded%20Burst%20Force%20Feedback%20Effect) "None" false false false)
    (if (Variables|Default|GetUseCombinedSkiddingNS)
      _
      (elif (Utilities|Enum|Equal(Enum) (Class|FortSurfaceTrackingComponent|GetCurrentSurfaceType (Footsteps|GetSurfaceTrackingComponent (Variables|FortMovementMode_BaseExtLogic|GetOwnerPawn))) "SurfaceType8")
        _
        (else
          (Default|ApplySkiddingGCNL)
          (CallFunction|RemoveWaterSkiingGCNL))))
    (else
      (Game|Feedback|LocalStopForceFeedback (Fortnite|GetFortPlayerControllerfromActor (Variables|FortMovementMode_BaseExtLogic|GetOwnerPawn)) (Variables|FF|GetLooping%20Force%20Feedback%20Effect) "None")
      (CallFunction|RemoveSkiddingGCNL))))

(event Custom|OnWaterStateChangeCustomEvent
  (params (bInWater :type Boolean))
  ;[0 0.1438 1] Start & Stop Force Feedback & Skidding GCNLs based on state
  (if (Variables|Default|GetUseCombinedSkiddingNS)
    _
    (elif bInWater
      (Default|ApplyWaterSkiingGCNL)
      (CallFunction|RemoveSkiddingGCNL)
      (else
        (CallFunction|RemoveWaterSkiingGCNL)))))

(event Custom|OnSurfaceChanged
  (params (SurfaceActor :type Actor%20Object%20Reference) (CurrentSurfaceType :type EPhysicalSurface%20Enum :default "SurfaceType_Default") (PrevSurfaceType :type EPhysicalSurface%20Enum :default "SurfaceType_Default"))
  ;[0 0.1438 1] Start & Stop Force Feedback & Skidding GCNLs based on state
  (if (Utilities|Enum|Equal(Enum) CurrentSurfaceType "SurfaceType8")
    (Default|ApplyWaterSkiingGCNL)
    (CallFunction|RemoveSkiddingGCNL)
    (else
      (CallFunction|RemoveWaterSkiingGCNL)
      (Default|ApplySkiddingGCNL))))

(fn AddNonMMEPushedTetheredTag
  (GameplayEffects|ApplyGameplayEffectToSelf :self (Ability|GetAbilitySystemComponent (Variables|FortMovementMode_BaseExtLogic|GetOwnerPawn)) :GameplayEffectClass (Variables|GameplayEffects|GetGE%20Tethered%20Any%20State) :Level 0.0))

(fn ApplyCombinedNSSkiddingGCNL
  (GameplayEffects|ApplyGameplayEffectToSelf :self (Ability|GetFortAbilitySystemComponent (Variables|FortMovementMode_BaseExtLogic|GetOwnerPawn)) :GameplayEffectClass (Variables|GameplayEffects|GetGE_Apply%20Combined%20SkiddingGCNL) :Level 1.0))

(fn ApplyMME_GE_AND_SpeedLineGCNL
  (GameplayEffects|ApplyGameplayEffectToSelf :self (Ability|GetFortAbilitySystemComponent (Variables|FortMovementMode_BaseExtLogic|GetOwnerPawn)) :GameplayEffectClass (Variables|GameplayEffects|GetGE_Apply_MME_GA_And_SpeedLinesGCNL) :Level 1.0))

(fn ApplySkiddingGCNL
  (GameplayEffects|ApplyGameplayEffectToSelf :self (Ability|GetFortAbilitySystemComponent (Variables|FortMovementMode_BaseExtLogic|GetOwnerPawn)) :GameplayEffectClass (Variables|GameplayEffects|GetGE_ApplySkiddingGCNL) :Level 1.0))

(fn ApplyWaterSkiingGCNL
  (GameplayEffects|ApplyGameplayEffectToSelf :self (Ability|GetFortAbilitySystemComponent (Variables|FortMovementMode_BaseExtLogic|GetOwnerPawn)) :GameplayEffectClass (Variables|GameplayEffects|GetGE_ApplyWaterSkiingGCNL) :Level 1.0))

(fn LinkAnimLayer
  (bind _ownerpawn (Variables|FortMovementMode_BaseExtLogic|GetOwnerPawn))
  (if (not (Networking|HasAuthority _ownerpawn))
    (Utilities|IsValid (Components|SkeletalMesh|AnimationBlueprintLinking|GetLinkedAnimLayerInstancebyClass (Variables|/Script/Engine.Character|GetMesh _ownerpawn) (Variables|Default|GetTetheredAnimLayer))
      (:Is%20Not%20Valid
        (Animation|LinkAnimLayer (Variables|FortMovementMode_BaseExtLogic|GetOwnerPawn) (Variables|Default|GetTetheredAnimLayer))))))

(fn RemoveCombinedNSSkiddingGCNL
  (GameplayEffects|RemoveActiveEffectswithGrantedTags (Ability|GetFortAbilitySystemComponent (Variables|FortMovementMode_BaseExtLogic|GetOwnerPawn)) (Variables|GameplayTags|GetCombinedNS%20SkiddingGCNL%20Tag)))

(fn RemoveNonMMEPushedTetheredtag
  (GameplayEffects|RemoveActiveEffectswithGrantedTags (Ability|GetFortAbilitySystemComponent (Variables|FortMovementMode_BaseExtLogic|GetOwnerPawn)) (Variables|GameplayTags|GetTethered%20Any%20State%20Tag)))

(fn RemoveSkiddingGCNL
  (GameplayEffects|RemoveActiveEffectswithGrantedTags (Ability|GetFortAbilitySystemComponent (Variables|FortMovementMode_BaseExtLogic|GetOwnerPawn)) (Variables|GameplayTags|GetSkiddingGCNL%20Tag)))

(fn RemoveWaterSkiingGCNL
  (GameplayEffects|RemoveActiveEffectswithGrantedTags (Ability|GetFortAbilitySystemComponent (Variables|FortMovementMode_BaseExtLogic|GetOwnerPawn)) (Variables|GameplayTags|GetWaterSkiingGCNL%20Tag)))

(fn RemoveMME_GEs_AND_GCNLs
  (GameplayEffects|RemoveActiveEffectswithGrantedTags (Ability|GetFortAbilitySystemComponent (Variables|FortMovementMode_BaseExtLogic|GetOwnerPawn)) (Variables|GameplayTags|GetSpeedLinesGCNL%20Tag)))

(fn UnlinkAnimLayer
  (bind _ownerpawn (Variables|FortMovementMode_BaseExtLogic|GetOwnerPawn))
  (if (not (Networking|HasAuthority _ownerpawn))
    (Animation|UnlinkAnimLayer _ownerpawn (Variables|Default|GetTetheredAnimLayer))))
