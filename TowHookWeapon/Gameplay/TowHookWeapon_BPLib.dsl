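; Blueprint function library of helpers for the Tow Hook weapon and its projectile.
; AttachPrjToVehicleServer attaches the projectile to the hit actor with authority, deriving its pose from a vehicle's TowHookAttach socket or hit-based fallbacks; a scalable-float toggle keeps the hit location.
; Other helpers recursively test attach parents for movement components and sphere-trace a safe spawn point.
; The ledge config and projectile radius getters show only empty or zero returns.
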(fn AttachPrjToVehicleServer
  (params (Hit%20Actor :type Actor%20Object%20Reference) (Hit :type Hit%20Result%20Structure) (Projectile :type Prj%20Athena%20Tow%20Hook%20Weapon%20Object%20Reference))
  (bind (_bblockinghit _binitialoverlap _time _distance _location _impactpoint _normal _impactnormal _physmat _hitactor _hitcomponent _hitbonename _bonename _hititem _elementindex _faceindex _tracestart _traceend) (Collision|BreakHitResult Hit))
  (bind _projectile (Variables|GetProjectile))
  (bind _returnvalue (Actor|GetComponentbyClass _projectile "/Script/FortniteGame.FortAttachableProjectileComponent"))
  (bind _hit_actor (Variables|GetHit%20Actor))
  (LocalVariables|SetHitLoc _location)
  (bind _output_get (LocalVariables|SetImpact%20Normal _normal))
  (Development|PrintString 0 (Utilities|String|Append "NORMAL:  " (Utilities|String|ToString(Vector) _output_get)) true true "(R=0.000000,G=0.660000,B=1.000000,A=1.000000)" 1e+01)
  (Utilities|Casting|CastToFortVehicleInterface _hit_actor
    (:then
      (bind _asfort_athena_vehicle (Utilities|Casting|CastToFortAthenaVehicle _hit_actor))
      (if (Transformation|DoesSocketExist (Variables|/Script/FortniteGame.FortAthenaVehicle|GetSkeletalMesh _asfort_athena_vehicle) "TowHookAttach")
        (if (ScalableFloat|GetValueasBool (LocalVariables|GetSF_HackDelayEnabled))
          (LocalVariables|SetFinalLocation (LocalVariables|GetHitLoc))
          (bind (_returnvalue_location _returnvalue_rotation) (Math|Transform|BreakTransform (Transformation|GetSocketTransform (Variables|/Script/FortniteGame.FortAthenaVehicle|GetSkeletalMesh _asfort_athena_vehicle) "TowHookAttach" "RTS_World")))
          (LocalVariables|SetFinalRotation _returnvalue_rotation)
          (else
            (LocalVariables|SetFinalLocation _returnvalue_location)
            (LocalVariables|SetFinalRotation _returnvalue_rotation)))
        (else
          ;[0.939 1 0.01239] No Socket case - temp fix for hit going to 0,0,0
          (LocalVariables|SetFinalLocation (LocalVariables|GetHitLoc))
          (LocalVariables|SetFinalRotation (Transformation|GetActorRotation (Variables|GetProjectile))))))
    (:CastFailed
      ;[0.9205 1 0] TODO: train
      (LocalVariables|SetFinalLocation (+ (LocalVariables|GetHitLoc) (* (Transformation|GetActorForwardVector (Variables|GetProjectile)) 2e+01)))
      (LocalVariables|SetFinalRotation (Math|Rotator|MakeRotator (.roll (Transformation|GetActorRotation (Variables|GetHit%20Actor))) (.pitch (Transformation|GetActorRotation (Variables|GetHit%20Actor))) (.yaw (Transformation|GetActorRotation (Variables|GetHit%20Actor)))))))
  (Utilities|IsValid _returnvalue
    (:Is%20Valid
      (Gameplay|AuthorityAttachtoActor _returnvalue (Collision|MakeHitResult _bblockinghit _binitialoverlap _time _distance (LocalVariables|GetFinalLocation) _impactpoint _normal _impactnormal _physmat _hitactor _hitcomponent _hitbonename _bonename _hititem _elementindex _faceindex _tracestart _traceend) (Variables|GetHit%20Actor) 0.0 0.0 0.75)
      (Transformation|SetActorRotation (Variables|GetProjectile) (LocalVariables|GetFinalRotation) true))))

(fn TowHookWeaponIsMovableActor
  (params (Actor :type Actor%20Object%20Reference) (bCheckAttachParent :type Boolean :default true))
  (Utilities|IsValid Actor
    (:Is%20Valid
      (if (or (Utilities|IsValid (Actor|GetComponentbyClass Actor "/Script/Engine.MovementComponent") :!class /Script/Engine.KismetSystemLibrary) (Utilities|IsValid (Actor|GetComponentbyClass Actor "/Script/FortniteGame.FortMovingPlatformComponent") :!class /Script/Engine.KismetSystemLibrary))
        (return true)
        (elif (Variables|GetbCheckAttachParent)
          (return (CallFunction|TowHookWeaponIsMovableActor (Actor|GetAttachParentActor (Variables|GetActor)) true (Variables|GetActor)))
          (else
            (return false)))))
    (:Is%20Not%20Valid
      (return false))))

(fn TowHookWeaponFindSafeProjectileSpawn
  (params (SourcePlayerPawn :type Fort%20Player%20Pawn%20Object%20Reference) (DesiredSpawnLocation :type Vector))
  (bind _radius (Internal|TowHookWeaponGetProjectileRadius 0))
  (bind (_outhit _returnvalue) (Collision|SphereTraceByProfile SourcePlayerPawn (Math|Vector|MakeVector (.x (Transformation|GetActorLocation SourcePlayerPawn)) (.y (Transformation|GetActorLocation SourcePlayerPawn)) (.z DesiredSpawnLocation)) DesiredSpawnLocation _radius "FortProjectileHitAllPawns" false (Utilities|Array|MakeArray (Variables|GetSourcePlayerPawn)) "None" true "(R=1.000000,G=0.000000,B=0.000000,A=1.000000)" "(R=0.000000,G=1.000000,B=0.000000,A=1.000000)" 15.0))
  (if _returnvalue
    (return (+ (.Location (Collision|BreakHitResult _outhit)) (* (.Normal (Collision|BreakHitResult _outhit)) _radius)))
    (else
      (return (Variables|GetDesiredSpawnLocation)))))

(fn TowHookWeaponGetLedgeConfig
  (return))

(fn TowHookWeaponGetProjectileRadius
  (return 0.0))
